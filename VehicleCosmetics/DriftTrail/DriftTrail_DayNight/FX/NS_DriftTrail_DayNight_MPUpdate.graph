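# Niagara system for the DayNight vehicle drift trail cosmetic, attached to the car's skeletal mesh.
# Three ribbon emitters trail backward, with game mode setting lifetime and some offsets and velocities.
# Ribbons write a sine/cosine vertical wave position; its camera-facing dot product feeds a lerp with equal endpoints, so has no effect.
# Socket-placed sprites and two rotating wheel-wrap mesh layers complete the look; painted-color override modules are disabled.
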NiagaraSystem /VehicleCosmetics/DriftTrail/DriftTrail_DayNight/FX/NS_DriftTrail_DayNight_MPUpdate
dependencies {
 usedDataInterfaces [
  &NiagaraDataInterfaceArrayFloat3
  &NiagaraDataInterfaceCamera
  &NiagaraDataInterfaceCurve
  &NiagaraDataInterfaceSkeletalMesh
 ]
 usedDynamicInputs []
 usedModules [
  &self:SystemSpawnScript.SystemScriptSource.SystemScriptGraph.NiagaraNodeAssignment_0.SetVariables_#1
  &FN_UnpackVehicleFlags
  /VehicleCosmetics/FX/Niagara/Modules/VC_SetVehicleSpeedAttribs
  &SystemState
 ]
}
emitters {
 NE_DayNight_Ribbon {
  emitterUpdateScript [
   &EmitterState(Life Cycle Mode=::NewEnumerator1, Loop Duration=1, UseLoopDelay=false)
   &VC_FilteredSocketCount()
   &VC_RateQualityHandler(High Spawn Rate=60, Low Spawn Rate=60, Medium Spawn Rate=60, Use Custom Rate?=true)
   &VC_SpawnRibbon_v1(Number of Ribbons=2, Spawn Time=1, SpawnRate=$Emitter.Spawn Rate)
  ]
  particleSpawnScript [
   &VC_PaintedVariantsOverride(
    DefaultColor=<rgba 0.3763 0.0976 0.6514 0.2>
    R1 Crimson:LinearColor=<rgba 0.4635 0 0.0126 1>
    R1 DefaultColor=<rgba 0.1651 0.2542 0.9131 0.2>
    R3 Black:LinearColor=<rgba 0.1719 0.1719 0.1719 1>
    R3 BurntSienna:LinearColor=<rgba 0.3177 0.1227 0.0556 1>
    R3 Cobalt:LinearColor=<rgba 0.0165 0.0249 0.474 1>
    R3 Crimson:LinearColor=<rgba 1 0.0365 0.0626 1>
    R3 DefaultColor:LinearColor=<rgba 0.9131 0.1746 0.1301 0.2>
    R3 ForestGreen:LinearColor=<rgba 0 0.4427 0.0224 1>
    R3 Grey:LinearColor=<rgba 0.4271 0.4271 0.4271 1>
    R3 Lime:LinearColor=<rgba 0.5453 0.8698 0.0076 1>
    R3 Orange:LinearColor=<rgba 0.7917 0.3402 0 1>
    R3 Pink:LinearColor=<rgba 1 0.0781 0.7743 1>
    R3 Purple:LinearColor=<rgba 0.3242 0.0078 0.7396 1>
    R3 Saffron:LinearColor=<rgba 0.8542 0.6632 0.0356 1>
    R3 SkyBlue:LinearColor=<rgba 0.0485 0.5326 0.776 1>
    R3 TitaniumWhite:LinearColor=<rgba 2 2 2 1>
    Use R1 Painted Overrides=true
    Use R3 Painted Overrides=true
   ) {enabled=false}
   &VC_InitializeRibbon_v1(
    Color=<rgba 0.3763 0.0976 0.6514 0.2>
    Custom Offset=true
    Enable Skeletal Position=true
    Lifetime=&MergeHelper_FloatSwitch(0=1, 1=1, 2=0.5, 3=?, Index=$User.GameModeID, Num=::NewEnumerator1)
    Local Vector Offset=&AddVectorToPosition(Position=<-35 0 -28.5>, Vector=Linked() {linkedVariable {name=User.DTCarOffset, type=NiagaraPosition}})
    Offset Amount=1
    Ribbon Width=20
   )
   &AddVelocity(Velocity=<-1000 0 0>) {module=AddVelocity001}
   /SetVariables(Particles.Blue:LinearColor=<rgba 0.1651 0.2542 0.9131 0.2>, Particles.Orange:LinearColor=<rgba 0.9131 0.1746 0.1301 0.2>) {isSetParameters=true}
  ]
  particleUpdateScript [
   &ParticleState()
   &InheritSourceMovement(
    Applied Position Scale=&VectorFromFloat(Value=&FloatFromCurve(FloatCurve=DataInterface() {propertyValues {Curve {keys [
          {interpMode=RCIM_Cubic, tangentMode=RCTM_Break, time=0, value=1}
          {arriveTangent=-2.9025, interpMode=RCIM_Cubic, tangentMode=RCTM_Break, time=1, value=0}
         ]}}}))
    ApplyForce=false
    ApplyPosition=true
    ApplyPositionFalloff=false
   )
   &SolveForcesAndVelocity(Clamp Velocity=false, Limit Acceleration=false)
   &ScaleRibbonWidth(Ribbon Width Scale=&FloatFromCurve(FloatCurve=DataInterface() {propertyValues {Curve {
        defaultValue=1
        keys [
         {arriveTangent=16.0927, interpMode=RCIM_Cubic, leaveTangent=16.0927, tangentMode=RCTM_User, time=0.02, value=0}
         {
          arriveTangent=1.4881
          arriveTangentWeight=0.1138
          interpMode=RCIM_Cubic
          leaveTangent=1.4881
          leaveTangentWeight=0.1138
          tangentMode=RCTM_User
          tangentWeightMode=RCTWM_WeightedBoth
          time=0.1314
          value=0.7101
         }
         {arriveTangent=-0.1523, interpMode=RCIM_Cubic, leaveTangent=-0.1523, tangentMode=RCTM_User, time=0.3925, value=0.6916}
         {arriveTangent=-2.4411, interpMode=RCIM_Cubic, leaveTangent=-2.4411, tangentMode=RCTM_User, time=0.6182, value=0.498}
         {arriveTangent=-0.7985, interpMode=RCIM_Cubic, leaveTangent=-0.7985, tangentMode=RCTM_User, time=0.75, value=0}
        ]
       }}}))
   &VC_UpdateRibbon_v1(Enable Skeletal Position=true)
   /SetVariables(Particles.DotProduct:NiagaraFloat=&DotProduct(
     Vector A=/Niagara/DynamicInputs/Math/TransformVectorByMatrix(
      Vector:Vector3f=<1 0 0>
      Transformation Matrix:NiagaraMatrix=$Engine.Owner.SystemLocalToWorld
      Transform Vector as Position:NiagaraBool=false
     )
     Vector B=&NormalizeVector(
      Vector To Normalize=&SubtractVectorFromPosition(
       Position=$Engine.Owner.Position
       Vector=&ReturnCameraPosition(Coordinate Space:ENiagaraCoordinateSpace=::Simulation, Camera Query:NiagaraDataInterfaceCamera=DataInterface())
      )
      Safe Normalize=true
      Safe Fallback Threshold=0
      Safe Fallback Vector=<0 0 0.001>
     )
    )) {isSetParameters=true}
   /SetVariables(Particles.CustomPosition_R1:NiagaraPosition=&AddVectorToPosition(
     Position=$Particles.Position
     Vector=&MakeVector(
      X=0
      Y=0
      Z=&MaskFloat(
       Mask Source=0
       Mask Parameter=?
       Mask=0
       Masked Float=&Cosine(
        Normalized Angle=&Add_Float(A=0, B=&Subtract_Float(A=$Emitter.NormalizedLoopAge, B=$Particles.NormalizedAge))
        Period=1
        Scale=&FloatFromCurve(
         FloatCurve=DataInterface() {propertyValues {Curve {keys [
             <at 0 0 RCIM_Cubic>
             {arriveTangent=4.1101, interpMode=RCIM_Cubic, leaveTangent=4.1101, tangentMode=RCTM_User, time=0.3443, value=0.71}
             {arriveTangent=5.777, interpMode=RCIM_Cubic, leaveTangent=5.777, time=0.5965, value=3.8}
             {
              arriveTangent=-0.5881
              arriveTangentWeight=0.5973
              interpMode=RCIM_Cubic
              leaveTangent=-0.5881
              leaveTangentWeight=0.5973
              tangentMode=RCTM_User
              time=0.7585
              value=3.103
             }
            ]}}}
         Scale Curve=&Lerp_Float(
          A=2
          B=2
          Alpha=&NormalizeFloat(Value=&Power(a=&Add_Float(A=$Particles.DotProduct, B=0.1), b=5), Min=0, Max=1)
         )
        )
        Bias=0
       )
       Passthrough Float=&Sine(
        Normalized Angle=&Add_Float(A=0, B=&Subtract_Float(A=$Emitter.NormalizedLoopAge, B=$Particles.NormalizedAge))
        Period=1
        Scale=&FloatFromCurve(
         FloatCurve=DataInterface() {propertyValues {Curve {keys [
             <at 0 0 RCIM_Cubic>
             {arriveTangent=4.1101, interpMode=RCIM_Cubic, leaveTangent=4.1101, tangentMode=RCTM_User, time=0.3299, value=0.71}
             {arriveTangent=4.6426, interpMode=RCIM_Cubic, leaveTangent=4.6426, time=0.6086, value=3.8}
             {
              arriveTangent=-0.5881
              arriveTangentWeight=0.5973
              interpMode=RCIM_Cubic
              leaveTangent=-0.5881
              leaveTangentWeight=0.5973
              tangentMode=RCTM_User
              time=0.7585
              value=2.7
             }
            ]}}}
         Scale Curve=&Lerp_Float(
          A=2
          B=2
          Alpha=&NormalizeFloat(Value=&Power(a=&Add_Float(A=$Particles.DotProduct, B=0.1), b=5), Min=0, Max=1)
         )
        )
        Bias=0
       )
      )
     )
    )) {isSetParameters=true}
  ]
  rendererClasses [&NiagaraRibbonRendererProperties]
  renderers [
   {rendererClass=&NiagaraRibbonRendererProperties}
   {rendererClass=&NiagaraRibbonRendererProperties, rendererIndex=1}
   {rendererClass=&NiagaraRibbonRendererProperties, rendererIndex=2}
  ]
 }
 NE_R1_Warning {
  emitterUpdateScript [
   &EmitterState()
   &VC_RateQualityHandler(High Spawn Rate=60, Low Spawn Rate=60, Medium Spawn Rate=60, Use Custom Rate?=true)
   &VC_SpawnRibbon_v1(Number of Ribbons=2, Spawn Time=-1, SpawnRate=$Emitter.Spawn Rate)
  ]
  particleSpawnScript [
   &VC_PaintedVariantsOverride(
    DefaultColor=<rgba 1 0.7991 0.3712 1>
    R1 Crimson:LinearColor=<rgba 0.4635 0 0.0126 1>
    R1 DefaultColor=<rgba 0.008 0.0003 0.1779 1>
    R3 Black:LinearColor=<rgba 0.1719 0.1719 0.1719 1>
    R3 BurntSienna:LinearColor=<rgba 0.3177 0.1227 0.0556 1>
    R3 Cobalt:LinearColor=<rgba 0.0165 0.0249 0.474 1>
    R3 Crimson:LinearColor=<rgba 1 0.0365 0.0626 1>
    R3 DefaultColor:LinearColor=<rgba 3 2.008 1.14 1>
    R3 ForestGreen:LinearColor=<rgba 0 0.4427 0.0224 1>
    R3 Grey:LinearColor=<rgba 0.4271 0.4271 0.4271 1>
    R3 Lime:LinearColor=<rgba 0.5453 0.8698 0.0076 1>
    R3 Orange:LinearColor=<rgba 0.7917 0.3402 0 1>
    R3 Pink:LinearColor=<rgba 1 0.0781 0.7743 1>
    R3 Purple:LinearColor=<rgba 0.3242 0.0078 0.7396 1>
    R3 Saffron:LinearColor=<rgba 0.8542 0.6632 0.0356 1>
    R3 SkyBlue:LinearColor=<rgba 0.0485 0.5326 0.776 1>
    R3 TitaniumWhite:LinearColor=<rgba 2 2 2 1>
    Use R1 Painted Overrides=true
    Use R3 Painted Overrides=true
   ) {enabled=false}
   &VC_InitializeRibbon_v1(
    Color=<rgba 1 0.7991 0.3712 1>
    Custom Offset=true
    Enable Skeletal Position=true
    Facing Axis:ENiagaraOrientationAxis=::ZAxis
    Flip Facing=true
    Lifetime=&MergeHelper_FloatSwitch(0=1, 1=1, 2=0.5, 3=?, Index=$System.GameModeID, Num=::NewEnumerator1)
    Local Vector Offset=&AddVectorToPosition(
     Position=&MergeHelper_VectorSwitch(0=<0 0 -25>, 1=<0 0 -30>, 2=<0 0 -38>, 3=?, Index=$System.GameModeID, Num=::NewEnumerator1)
     Vector=$User.offset
    )
    Offset Amount=1
    Ribbon Width=30
   )
   &AddVelocity(Velocity=&MergeHelper_VectorSwitch(0=<-900 0 0>, 1=<-900 0 0>, 2=<-1000 0 0>, 3=?, Index=$System.GameModeID, Num=::NewEnumerator1)) {module=AddVelocity001}
   /SetVariables(Particles.Position=&AddVectorToPosition(
     Position=$Particles.Position
     Vector=&TransformVector(
      Vector=&MakeVector(X=-7.636, Y=&MaskFloatBySpawnGroup(Masked Float=0, Passthrough Float=0), Z=0)
      Destination Space=::Simulation
     )
    )) {isSetParameters=true, module=SetVariables_#2}
  ]
  particleUpdateScript [
   &FN_DisableTireFX(Mask Index Attribute:NiagaraInt32=$Particles.SpawnGroup, Mask Index FL=2, Mask Index FR=3, Mask Index RL=0, Mask Index RR=1)
   &InheritSourceMovement(
    Applied Position Scale=&Multiply_VectorByFloat(Vector=<1 1 1>, Float=1)
    ApplyForce=false
    ApplyPosition=true
    ApplyPositionFalloff=false
   )
   &ParticleState()
   &SolveForcesAndVelocity(Clamp Velocity=false, Limit Acceleration=false)
   &ScaleColor(ScaleA=true, ScaleRGB=true) {enabled=false}
   /SetVariables(Particles.DotProduct:NiagaraFloat=&DotProduct(
     Vector A=&TransformVectorByMatrix(
      Vector:Vector3f=<1 0 0>
      Transformation Matrix:NiagaraMatrix=$Engine.Owner.SystemLocalToWorld
      Transform Vector as Position:NiagaraBool=false
     )
     Vector B=&NormalizeVector(
      Vector To Normalize=&SubtractVectorFromPosition(
       Position=$Engine.Owner.Position
       Vector=&ReturnCameraPosition(Coordinate Space:ENiagaraCoordinateSpace=::Simulation, Camera Query:NiagaraDataInterfaceCamera=DataInterface())
      )
      Safe Normalize=true
      Safe Fallback Threshold=0
      Safe Fallback Vector=<0 0 0.001>
     )
    )) {isSetParameters=true, module=SetVariables_#3}
   /SetVariables(Particles.CustomPosition_R1:NiagaraPosition=&AddVectorToPosition(
     Position=$Particles.Position
     Vector=&MakeVector(
      X=0
      Y=0
      Z=&MaskFloat(
       Mask Source=0
       Mask Parameter=?
       Mask=0
       Masked Float=&Cosine(
        Normalized Angle=&Add_Float(A=0, B=&Subtract_Float(A=$Emitter.NormalizedLoopAge, B=$Particles.NormalizedAge))
        Period=1
        Scale=&FloatFromCurve(
         FloatCurve=DataInterface() {propertyValues {Curve {keys [
             <at 0 0 RCIM_Cubic>
             {arriveTangent=4.1101, interpMode=RCIM_Cubic, leaveTangent=4.1101, tangentMode=RCTM_User, time=0.3443, value=0.71}
             {arriveTangent=5.777, interpMode=RCIM_Cubic, leaveTangent=5.777, time=0.5965, value=3.8}
             {
              arriveTangent=-0.5881
              arriveTangentWeight=0.5973
              interpMode=RCIM_Cubic
              leaveTangent=-0.5881
              leaveTangentWeight=0.5973
              tangentMode=RCTM_User
              time=0.7585
              value=3.103
             }
            ]}}}
         Scale Curve=&Lerp_Float(
          A=2
          B=2
          Alpha=&NormalizeFloat(Value=&Power(a=&Add_Float(A=$Particles.DotProduct, B=0.1), b=5), Min=0, Max=1)
         )
        )
        Bias=0
       )
       Passthrough Float=&Sine(
        Normalized Angle=&Add_Float(A=0, B=&Subtract_Float(A=$Emitter.NormalizedLoopAge, B=$Particles.NormalizedAge))
        Period=1
        Scale=&FloatFromCurve(
         FloatCurve=DataInterface() {propertyValues {Curve {keys [
             <at 0 0 RCIM_Cubic>
             {arriveTangent=4.1101, interpMode=RCIM_Cubic, leaveTangent=4.1101, tangentMode=RCTM_User, time=0.3299, value=0.71}
             {arriveTangent=4.6426, interpMode=RCIM_Cubic, leaveTangent=4.6426, time=0.6086, value=3.8}
             {
              arriveTangent=-0.5881
              arriveTangentWeight=0.5973
              interpMode=RCIM_Cubic
              leaveTangent=-0.5881
              leaveTangentWeight=0.5973
              tangentMode=RCTM_User
              time=0.7585
              value=2.7
             }
            ]}}}
         Scale Curve=&Lerp_Float(
          A=2
          B=2
          Alpha=&NormalizeFloat(Value=&Power(a=&Add_Float(A=$Particles.DotProduct, B=0.1), b=5), Min=0, Max=1)
         )
        )
        Bias=0
       )
      )
     )
    )) {isSetParameters=true, module=SetVariables_#4}
   &ScaleRibbonWidth(Ribbon Width Scale=&FloatFromCurve(
     FloatCurve=DataInterface() {propertyValues {Curve {
        defaultValue=0.8294
        keys [
         {tangentMode=RCTM_User, time=0, value=0.3}
         {arriveTangent=-0.0405, interpMode=RCIM_Cubic, leaveTangent=-0.0405, tangentMode=RCTM_User, time=0.5, value=1}
         {arriveTangent=-8.0089, interpMode=RCIM_Cubic, leaveTangent=-8.0089, tangentMode=RCTM_User, time=0.9442, value=0}
        ]
       }}}
     Scale Curve=1.4
    ))
   &VC_UpdateRibbon_v1(Detach Locked Particle=true, Enable Skeletal Position=true)
  ]
  rendererClasses [&NiagaraRibbonRendererProperties]
  renderers [{rendererClass=&NiagaraRibbonRendererProperties}]
 }
 NE_R1_Warning001 {
  emitterUpdateScript [
   &EmitterState()
   &VC_RateQualityHandler(High Spawn Rate=60, Low Spawn Rate=60, Medium Spawn Rate=60, Use Custom Rate?=true)
   &VC_SpawnRibbon_v1(Number of Ribbons=2, Spawn Time=-1, SpawnRate=$Emitter.Spawn Rate)
  ]
  particleSpawnScript [
   &VC_PaintedVariantsOverride(
    BurntSienna=<rgba 0.0642 0.0198 0.0079 1>
    DefaultColor=<rgba 0.0044 0.0003 0.0931 1>
    R1 Crimson:LinearColor=<rgba 0.4635 0 0.0126 1>
    R1 DefaultColor=<rgba 0.008 0.0003 0.1779 1>
    R3 Black:LinearColor=<rgba 0.1719 0.1719 0.1719 1>
    R3 BurntSienna:LinearColor=<rgba 0.3177 0.1227 0.0556 1>
    R3 Cobalt:LinearColor=<rgba 0.0165 0.0249 0.474 1>
    R3 Crimson:LinearColor=<rgba 1 0.0365 0.0626 1>
    R3 DefaultColor:LinearColor=<rgba 3 2.008 1.14 1>
    R3 ForestGreen:LinearColor=<rgba 0 0.4427 0.0224 1>
    R3 Grey:LinearColor=<rgba 0.4271 0.4271 0.4271 1>
    R3 Lime:LinearColor=<rgba 0.5453 0.8698 0.0076 1>
    R3 Orange:LinearColor=<rgba 0.7917 0.3402 0 1>
    R3 Pink:LinearColor=<rgba 1 0.0781 0.7743 1>
    R3 Purple:LinearColor=<rgba 0.3242 0.0078 0.7396 1>
    R3 Saffron:LinearColor=<rgba 0.8542 0.6632 0.0356 1>
    R3 SkyBlue:LinearColor=<rgba 0.0485 0.5326 0.776 1>
    R3 TitaniumWhite:LinearColor=<rgba 2 2 2 1>
    Use R1 Painted Overrides=true
    Use R3 Painted Overrides=true
   ) {enabled=false}
   &VC_InitializeRibbon_v1(
    Color=<rgba 0.0044 0.0003 0.0931 1>
    Custom Offset=true
    Enable Skeletal Position=true
    Facing Axis:ENiagaraOrientationAxis=::ZAxis
    Flip Facing=true
    Lifetime=&MergeHelper_FloatSwitch(0=1, 1=1, 2=0.5, 3=?, Index=$System.GameModeID, Num=::NewEnumerator1)
    Local Vector Offset=&AddVectorToPosition(
     Position=&MergeHelper_VectorSwitch(0=<0 0 -25>, 1=<0 0 -30>, 2=<0 0 -30>, 3=?, Index=$System.GameModeID, Num=::NewEnumerator1)
     Vector=$User.offset
    )
    Offset Amount=1
    Ribbon Width=30
   )
   &AddVelocity(Velocity=&MergeHelper_VectorSwitch(0=<-900 0 0>, 1=<-900 0 0>, 2=<-1000 0 0>, 3=?, Index=$System.GameModeID, Num=::NewEnumerator1)) {module=AddVelocity001}
   /SetVariables(
    Particles.OutlineColor:LinearColor=<rgba 3 2.008 1.14 1>
    Particles.Position=&AddVectorToPosition(
     Position=$Particles.Position
     Vector=&TransformVector(
      Vector=&MakeVector(X=-7.636, Y=&MaskFloatBySpawnGroup(Masked Float=0, Passthrough Float=0), Z=0)
      Destination Space=::Simulation
     )
    )
   ) {isSetParameters=true, module=SetVariables_#2}
  ]
  particleUpdateScript [
   &FN_DisableTireFX(Mask Index Attribute:NiagaraInt32=$Particles.SpawnGroup, Mask Index FL=2, Mask Index FR=3, Mask Index RL=0, Mask Index RR=1)
   &InheritSourceMovement(
    Applied Position Scale=&Multiply_VectorByFloat(Vector=<1 1 1>, Float=1)
    ApplyForce=false
    ApplyPosition=true
    ApplyPositionFalloff=false
   )
   &ParticleState()
   &SolveForcesAndVelocity(Clamp Velocity=false, Limit Acceleration=false)
   &ScaleColor(ScaleA=true, ScaleRGB=true) {enabled=false}
   /SetVariables(Particles.DotProduct:NiagaraFloat=&DotProduct(
     Vector A=&TransformVectorByMatrix(
      Vector:Vector3f=<1 0 0>
      Transformation Matrix:NiagaraMatrix=$Engine.Owner.SystemLocalToWorld
      Transform Vector as Position:NiagaraBool=false
     )
     Vector B=&NormalizeVector(
      Vector To Normalize=&SubtractVectorFromPosition(
       Position=$Engine.Owner.Position
       Vector=&ReturnCameraPosition(Coordinate Space:ENiagaraCoordinateSpace=::Simulation, Camera Query:NiagaraDataInterfaceCamera=DataInterface())
      )
      Safe Normalize=true
      Safe Fallback Threshold=0
      Safe Fallback Vector=<0 0 0.001>
     )
    )) {isSetParameters=true, module=SetVariables_#3}
   /SetVariables(Particles.CustomPosition_R1:NiagaraPosition=&AddVectorToPosition(
     Position=$Particles.Position
     Vector=&MakeVector(
      X=0
      Y=0
      Z=&MaskFloat(
       Mask Source=0
       Mask Parameter=?
       Mask=0
       Masked Float=&Cosine(
        Normalized Angle=&Add_Float(A=0, B=&Subtract_Float(A=$Emitter.NormalizedLoopAge, B=$Particles.NormalizedAge))
        Period=1
        Scale=&FloatFromCurve(
         FloatCurve=DataInterface() {propertyValues {Curve {keys [
             <at 0 0 RCIM_Cubic>
             {arriveTangent=4.1101, interpMode=RCIM_Cubic, leaveTangent=4.1101, tangentMode=RCTM_User, time=0.3443, value=0.71}
             {arriveTangent=5.777, interpMode=RCIM_Cubic, leaveTangent=5.777, time=0.5965, value=3.8}
             {
              arriveTangent=-0.5881
              arriveTangentWeight=0.5973
              interpMode=RCIM_Cubic
              leaveTangent=-0.5881
              leaveTangentWeight=0.5973
              tangentMode=RCTM_User
              time=0.7585
              value=3.103
             }
            ]}}}
         Scale Curve=&Lerp_Float(
          A=2
          B=2
          Alpha=&NormalizeFloat(Value=&Power(a=&Add_Float(A=$Particles.DotProduct, B=0.1), b=5), Min=0, Max=1)
         )
        )
        Bias=0
       )
       Passthrough Float=&Sine(
        Normalized Angle=&Add_Float(A=0, B=&Subtract_Float(A=$Emitter.NormalizedLoopAge, B=$Particles.NormalizedAge))
        Period=1
        Scale=&FloatFromCurve(
         FloatCurve=DataInterface() {propertyValues {Curve {keys [
             <at 0 0 RCIM_Cubic>
             {arriveTangent=4.1101, interpMode=RCIM_Cubic, leaveTangent=4.1101, tangentMode=RCTM_User, time=0.3299, value=0.71}
             {arriveTangent=4.6426, interpMode=RCIM_Cubic, leaveTangent=4.6426, time=0.6086, value=3.8}
             {
              arriveTangent=-0.5881
              arriveTangentWeight=0.5973
              interpMode=RCIM_Cubic
              leaveTangent=-0.5881
              leaveTangentWeight=0.5973
              tangentMode=RCTM_User
              time=0.7585
              value=2.7
             }
            ]}}}
         Scale Curve=&Lerp_Float(
          A=2
          B=2
          Alpha=&NormalizeFloat(Value=&Power(a=&Add_Float(A=$Particles.DotProduct, B=0.1), b=5), Min=0, Max=1)
         )
        )
        Bias=0
       )
      )
     )
    )) {isSetParameters=true, module=SetVariables_#4}
   &ScaleRibbonWidth(Ribbon Width Scale=&FloatFromCurve(
     FloatCurve=DataInterface() {propertyValues {Curve {
        defaultValue=0.8294
        keys [
         {tangentMode=RCTM_User, time=0, value=0.5}
         {arriveTangent=-0.0405, interpMode=RCIM_Cubic, leaveTangent=-0.0405, tangentMode=RCTM_User, time=0.5, value=1}
         {arriveTangent=-8.0089, interpMode=RCIM_Cubic, leaveTangent=-8.0089, tangentMode=RCTM_User, time=0.95, value=0}
        ]
       }}}
     Scale Curve=1.2
    ))
   &VC_UpdateRibbon_v1(Detach Locked Particle=true, Enable Skeletal Position=true)
  ]
  rendererClasses [&NiagaraRibbonRendererProperties]
  renderers [
   {rendererClass=&NiagaraRibbonRendererProperties}
   {rendererClass=&NiagaraRibbonRendererProperties, rendererIndex=1}
   {rendererClass=&NiagaraRibbonRendererProperties, rendererIndex=2}
  ]
 }
 NE_VC_DriftTrail_Sprite_Parent {
  emitterUpdateScript [
   &EmitterState(
    Inactive Response=::NewEnumerator1
    Life Cycle Mode=::NewEnumerator1
    Loop Behavior=::NewEnumerator1
    Loop Duration Mode=::NewEnumerator1
    UseLoopDelay=false
   )
   &SpawnBurst_Instantaneous(Spawn Count=2, Use Spawn Probability=false, UseLoopCountLimit=false)
   &SpawnBurst_Instantaneous(Spawn Count=2, Spawn Group=1, Use Spawn Probability=false, UseLoopCountLimit=false) {module=SpawnBurst_Instantaneous002}
   &SpawnRate(SpawnRate=50, Use Spawn Probability=false) {enabled=false}
  ]
  particleSpawnScript [
   &VC_PaintedVariantsOverride() {enabled=false}
   /SetVariables(Particles.VisibilityTag=$Emitter.SpawnGroup) {isSetParameters=true}
   &InitializeParticle(
    Color=<rgba 1 0.7991 0.3712 1>
    Lifetime=$Emitter.CurrentLoopDuration
    Position Mode=::NewEnumerator0
    Sprite Size Mode=::NewEnumerator3
    Uniform Sprite Size=/VehicleCosmetics/FX/Niagara/Modules/MergeEmitters/VC_FloatBySpawnGroup(
     SpawnGroup or ExecIndex:NiagaraInt32=0
     CustomInt:NiagaraInt32=0
     MaximumSpawnGroups:NiagaraInt32=3
     Float SG 0:NiagaraFloat=280
     Float SG 1:NiagaraFloat=30
     Float SG 2:NiagaraFloat=30
     Float SG 3:NiagaraFloat=0
     Float SG 4:NiagaraFloat=0
     Float SG 5:NiagaraFloat=0
    )
    Write Material Random=true
   )
   &PartitionParticles()
   &SkeletalMeshLocation(
    Bone / Socket Index=&Add_Integer(A=$Particles.PartitionParticles.Partition, B=4)
    Bone / Socket Sampling Mode=::NewEnumerator5
    Interpolate Mesh Sampling=false
    MaskSpawnGroup=false
    Mesh Sampling Type=::NewEnumerator6
    Sampled Position Offset=&AddVectorToPosition(
     Position=$User.DTCarOffset
     Vector=&MergeHelper_VectorSwitch(0=<-55 0 -25>, 1=<-55 0 -25>, 2=<-55 0 -25>, 3=?, Index=$User.GameModeID, Num=::NewEnumerator1)
    )
    Skeletal Mesh=$System.SkeletalMesh
   )
   &AddVelocity(Velocity=<-1500 0 0>) {enabled=false}
  ]
  particleUpdateScript [
   &ParticleState()
   &ScaleSpriteSize(
    Scale Sprite Size Mode=::NewEnumerator0
    Uniform Scale Factor=&VC_FloatBySpawnGroup(
     SpawnGroup or ExecIndex:NiagaraInt32=0
     CustomInt:NiagaraInt32=0
     MaximumSpawnGroups:NiagaraInt32=3
     Float SG 0:NiagaraFloat=&FloatFromCurve(
      FloatCurve=DataInterface() {propertyValues {Curve {keys [{tangentMode=RCTM_User, time=0, value=0.2}, {tangentMode=RCTM_User, time=0.85, value=2}, <at 1 0>]}}}
      Scale Curve=1.8
     )
     Float SG 1:NiagaraFloat=0
     Float SG 2:NiagaraFloat=1.2
     Float SG 3:NiagaraFloat=0
     Float SG 4:NiagaraFloat=0
     Float SG 5:NiagaraFloat=0
    )
   )
   &SolveForcesAndVelocity(Clamp Velocity=false, Limit Acceleration=false)
  ]
  rendererClasses [&NiagaraSpriteRendererProperties]
  renderers [
   {rendererClass=&NiagaraSpriteRendererProperties}
   {rendererClass=&NiagaraSpriteRendererProperties, rendererIndex=1}
  ]
 }
 NE_WheelWrap {
  emitterUpdateScript [
   &EmitterState(
    Inactive Response=::NewEnumerator1
    Life Cycle Mode=::NewEnumerator1
    Loop Behavior=::NewEnumerator1
    Loop Duration Mode=::NewEnumerator1
    UseLoopDelay=false
   )
   &SpawnBurst_Instantaneous(Spawn Count=2, Use Spawn Probability=false, UseLoopCountLimit=false)
  ]
  particleSpawnScript [
   &VC_PaintedVariantsOverride(
    DefaultColor=<rgba 0.0027 0.0003 0.0578 1>
    R2 DefaultColor:LinearColor=<rgba 1 0.7991 0.3712 1>
    R3 Black:LinearColor=<rgba 0.1719 0.1719 0.1719 1>
    R3 BurntSienna:LinearColor=<rgba 0.3177 0.1227 0.0556 1>
    R3 Cobalt:LinearColor=<rgba 0.0165 0.0249 0.474 1>
    R3 Crimson:LinearColor=<rgba 1 0.0365 0.0626 1>
    R3 DefaultColor:LinearColor=<rgba 3 2.008 1.14 1>
    R3 ForestGreen:LinearColor=<rgba 0 0.4427 0.0224 1>
    R3 Grey:LinearColor=<rgba 0.4271 0.4271 0.4271 1>
    R3 Lime:LinearColor=<rgba 0.5453 0.8698 0.0076 1>
    R3 Orange:LinearColor=<rgba 0.7917 0.3402 0 1>
    R3 Pink:LinearColor=<rgba 1 0.0781 0.7743 1>
    R3 Purple:LinearColor=<rgba 0.3242 0.0078 0.7396 1>
    R3 Saffron:LinearColor=<rgba 0.8542 0.6632 0.0356 1>
    R3 SkyBlue:LinearColor=<rgba 0.0485 0.5326 0.776 1>
    R3 TitaniumWhite:LinearColor=<rgba 2 2 2 1>
    Use R2 Painted Overrides=true
    Use R3 Painted Overrides=true
   ) {enabled=false}
   &InitializeParticle(
    Color=<rgba 0.0027 0.0003 0.0578 1>
    Lifetime=$Emitter.CurrentLoopDuration
    Mesh Scale Mode=::NewEnumerator4
    Sprite Rotation Mode=::NewEnumerator2
    Sprite Size Mode=::NewEnumerator3
    UsePositionOffset=false
    Write Material Random=true
   )
   &PartitionParticles()
   &InitialMeshOrientation(
    Mesh Orientation Mode=::NewEnumerator7
    Rotation=&SelectVectorFromArray(
     Direct Array Index=$Particles.PartitionParticles.Partition
     Vector Selection Array=DataInterface() {propertyValues {FloatData [<0 0 0>, <-0.5 0 0>]}}
    )
    Use Rotation Vector=true
   )
   &SkeletalMeshLocation(
    Bone / Socket Index=&Add_Integer(A=$Particles.PartitionParticles.Partition, B=4)
    Bone / Socket Sampling Mode=::NewEnumerator5
    Interpolate Mesh Sampling=false
    MaskSpawnGroup=false
    Mesh Sampling Type=::NewEnumerator6
    Skeletal Mesh=$System.SkeletalMesh
   )
  ]
  particleUpdateScript [
   &FN_DisableTireFX(
    Mask Index Attribute:NiagaraInt32=$Particles.PartitionParticles.Partition
    Mask Index FL=2
    Mask Index FR=3
    Mask Index RL=0
    Mask Index RR=1
   )
   &VC_WheelWrapScale(Wheel Index Attribute:NiagaraInt32=&Add_Integer(A=$Particles.PartitionParticles.Partition, B=2))
   &ParticleState(Kill Particles When Lifetime Has Elapsed=false, Let Infinitely Lived Particles Die When Emitter Deactivates=true)
   &UpdateMeshOrientation(Rotation Coordinate Space=::NewEnumerator0, Rotation Rate=0.5, Rotation Vector=<0 0.8 0>)
   &SkeletalMeshLocation(
    Bone / Socket Index=&Add_Integer(A=$Particles.PartitionParticles.Partition, B=4)
    Bone / Socket Sampling Mode=::NewEnumerator5
    MaskSpawnGroup=false
    Mesh Sampling Type=::NewEnumerator6
    Orientation Sampling=::NewEnumerator1
    Sampled Position Offset=&SelectVectorFromArray(
     Direct Array Index=$Particles.PartitionParticles.Partition
     Vector Selection Array=DataInterface() {propertyValues {FloatData [<0 0 0>, <0 0 0>]}}
    )
    Skeletal Mesh=$System.SkeletalMesh
   ) {module=SkeletalMeshLocation001}
   &SolveForcesAndVelocity(Clamp Velocity=false, Limit Acceleration=false)
   /SetVariables(
    Particles.GalaxyColor:LinearColor=<rgba 1 0.7991 0.3712 1>
    Particles.OutlineColor:LinearColor=<rgba 3 2.008 1.14 1>
    Particles.ScaleGalaxy:Vector3f=<1 0.23 1>
   ) {isSetParameters=true, module=SetVariables_#5}
  ]
  rendererClasses [&NiagaraMeshRendererProperties]
  renderers [
   {rendererClass=&NiagaraMeshRendererProperties, rendererIndex=0}
   {rendererClass=&NiagaraMeshRendererProperties, rendererIndex=1}
   {rendererClass=&NiagaraMeshRendererProperties, rendererIndex=2}
   {rendererClass=&NiagaraMeshRendererProperties, rendererIndex=3}
  ]
 }
 NE_WheelWrap001 {
  emitterUpdateScript [
   &EmitterState(
    Inactive Response=::NewEnumerator1
    Life Cycle Mode=::NewEnumerator1
    Loop Behavior=::NewEnumerator1
    Loop Duration Mode=::NewEnumerator1
    UseLoopDelay=false
   )
   &SpawnBurst_Instantaneous(Spawn Count=2, Use Spawn Probability=false, UseLoopCountLimit=false)
  ]
  particleSpawnScript [
   &VC_PaintedVariantsOverride(
    DefaultColor=<rgba 0.0027 0.0003 0.0578 1>
    R2 DefaultColor:LinearColor=<rgba 1 0.7991 0.3712 1>
    R3 Black:LinearColor=<rgba 0.1719 0.1719 0.1719 1>
    R3 BurntSienna:LinearColor=<rgba 0.3177 0.1227 0.0556 1>
    R3 Cobalt:LinearColor=<rgba 0.0165 0.0249 0.474 1>
    R3 Crimson:LinearColor=<rgba 1 0.0365 0.0626 1>
    R3 DefaultColor:LinearColor=<rgba 3 2.008 1.14 1>
    R3 ForestGreen:LinearColor=<rgba 0 0.4427 0.0224 1>
    R3 Grey:LinearColor=<rgba 0.4271 0.4271 0.4271 1>
    R3 Lime:LinearColor=<rgba 0.5453 0.8698 0.0076 1>
    R3 Orange:LinearColor=<rgba 0.7917 0.3402 0 1>
    R3 Pink:LinearColor=<rgba 1 0.0781 0.7743 1>
    R3 Purple:LinearColor=<rgba 0.3242 0.0078 0.7396 1>
    R3 Saffron:LinearColor=<rgba 0.8542 0.6632 0.0356 1>
    R3 SkyBlue:LinearColor=<rgba 0.0485 0.5326 0.776 1>
    R3 TitaniumWhite:LinearColor=<rgba 2 2 2 1>
    Use R2 Painted Overrides=true
    Use R3 Painted Overrides=true
   ) {enabled=false}
   &InitializeParticle(
    Color=<rgba 0.0027 0.0003 0.0578 1>
    Lifetime=$Emitter.CurrentLoopDuration
    Mesh Scale Mode=::NewEnumerator4
    Sprite Rotation Mode=::NewEnumerator2
    Sprite Size Mode=::NewEnumerator3
    UsePositionOffset=false
    Write Material Random=true
   )
   &PartitionParticles()
   &InitialMeshOrientation(
    Mesh Orientation Mode=::NewEnumerator7
    Rotation=&SelectVectorFromArray(
     Direct Array Index=$Particles.PartitionParticles.Partition
     Vector Selection Array=DataInterface() {propertyValues {FloatData [<0 0 0>, <-0.5 0 0>]}}
    )
    Use Rotation Vector=true
   )
   &SkeletalMeshLocation(
    Bone / Socket Index=&Add_Integer(A=$Particles.PartitionParticles.Partition, B=4)
    Bone / Socket Sampling Mode=::NewEnumerator5
    Interpolate Mesh Sampling=false
    MaskSpawnGroup=false
    Mesh Sampling Type=::NewEnumerator6
    Skeletal Mesh=$System.SkeletalMesh
   )
  ]
  particleUpdateScript [
   &FN_DisableTireFX(
    Mask Index Attribute:NiagaraInt32=$Particles.PartitionParticles.Partition
    Mask Index FL=2
    Mask Index FR=3
    Mask Index RL=0
    Mask Index RR=1
   )
   &VC_WheelWrapScale(Wheel Index Attribute:NiagaraInt32=&Add_Integer(A=$Particles.PartitionParticles.Partition, B=2))
   &ParticleState(Kill Particles When Lifetime Has Elapsed=false, Let Infinitely Lived Particles Die When Emitter Deactivates=true)
   &UpdateMeshOrientation(Rotation Coordinate Space=::NewEnumerator0, Rotation Rate=0.5, Rotation Vector=<0 0.8 0>)
   &SkeletalMeshLocation(
    Bone / Socket Index=&Add_Integer(A=$Particles.PartitionParticles.Partition, B=4)
    Bone / Socket Sampling Mode=::NewEnumerator5
    MaskSpawnGroup=false
    Mesh Sampling Type=::NewEnumerator6
    Orientation Sampling=::NewEnumerator1
    Sampled Position Offset=&SelectVectorFromArray(
     Direct Array Index=$Particles.PartitionParticles.Partition
     Vector Selection Array=DataInterface() {propertyValues {FloatData [<0 0 0>, <0 0 0>]}}
    )
    Skeletal Mesh=$System.SkeletalMesh
   ) {module=SkeletalMeshLocation001}
   &SolveForcesAndVelocity(Clamp Velocity=false, Limit Acceleration=false)
   /SetVariables(
    Particles.GalaxyColor:LinearColor=<rgba 0.9131 0.1746 0.1301 0.15>
    Particles.OutlineColor:LinearColor=<rgba 0.1651 0.2542 0.9131 0.15>
    Particles.ScaleGalaxy:Vector3f=<1 0.23 1>
    Particles.StarsColor:LinearColor=<rgba 0.3763 0.0976 0.6514 0.15>
   ) {isSetParameters=true, module=SetVariables_#5}
  ]
  rendererClasses [&NiagaraMeshRendererProperties]
  renderers [
   {rendererClass=&NiagaraMeshRendererProperties, rendererIndex=0}
   {rendererClass=&NiagaraMeshRendererProperties, rendererIndex=1}
   {rendererClass=&NiagaraMeshRendererProperties, rendererIndex=2}
   {rendererClass=&NiagaraMeshRendererProperties, rendererIndex=3}
  ]
 }
}
summary {
 systemName=NS_DriftTrail_DayNight_MPUpdate
 emitters [
  {emitterName=NE_WheelWrap, rendererClasses [&NiagaraMeshRendererProperties]}
  {emitterName=NE_R1_Warning001, rendererClasses [&NiagaraRibbonRendererProperties]}
  {emitterName=NE_R1_Warning, rendererClasses [&NiagaraRibbonRendererProperties]}
  {emitterName=NE_DayNight_Ribbon, rendererClasses [&NiagaraRibbonRendererProperties]}
  {emitterName=NE_VC_DriftTrail_Sprite_Parent, rendererClasses [&NiagaraSpriteRendererProperties]}
  {emitterName=NE_WheelWrap001, rendererClasses [&NiagaraMeshRendererProperties]}
 ]
 userVariables [
  User.bIsFrontendPreview {type=NiagaraBool}
  User.bIsPainted {type=NiagaraBool}
  User.ColorVariantID {type=NiagaraInt32}
  User.DTCarOffset {type=NiagaraPosition, description="", defaultValue {}}
  User.GameModeID {type=NiagaraInt32}
  User.offset {type=Vector3f, description="offset from wheels(set through cosmetic template)", defaultValue {struct=Vector3f, value=<0 0 0>}}
  User.PaintedColor {type=LinearColor, description="", defaultValue {struct=LinearColor, value=<rgba 1 1 1 1>}}
  User.PaintedScalar {type=NiagaraFloat, defaultValue=1, description=""}
  User.VehicleFlags {type=NiagaraInt32, defaultValue=806289408}
  User.Wheel_Radius_Array {
   type=NiagaraDataInterfaceArrayFloat
   defaultValue {
    struct=NiagaraExt_VariableValue_DataInterface
    value {
     dataInterface=&self:NiagaraDataInterfaceArrayFloat_1
     dataInterfaceClass=&NiagaraDataInterfaceArrayFloat
    }
   }
  }
  User.Wheel_Width_Array {
   type=NiagaraDataInterfaceArrayFloat
   defaultValue {
    struct=NiagaraExt_VariableValue_DataInterface
    value {
     dataInterface=&self:NiagaraDataInterfaceArrayFloat_0
     dataInterfaceClass=&NiagaraDataInterfaceArrayFloat
    }
   }
  }
  User.WheelTouchingArray {
   type=NiagaraDataInterfaceArrayFloat
   defaultValue {
    struct=NiagaraExt_VariableValue_DataInterface
    value {
     dataInterface=&self:NiagaraDataInterfaceArrayFloat_2
     dataInterfaceClass=&NiagaraDataInterfaceArrayFloat
    }
   }
  }
 ]
}
